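; User option component for the Creative/UEFN Experience Settings device that switches the new CH7 HUD on or off.
; On option changes, it writes the CreativeGameUINewHUD scalable float from ExperienceSettingsComponent defaults, using a force-enable-in-edit flag or bUseLatestHUD.
; The disable branch calls RefreshHUD, which finds the current minigame's NewHUD mutator through its mutator list and triggers a refresh.
; The options-loaded event is empty.
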(event UserOptions|EventHandleUserOptionsChanged
  (params (ChangedPropertyNames :type Array%20of%20Names))
  (CallFunction|HandleOptionsChanged))

(event UserOptions|EventHandleUserOptionsLoaded)

(fn HandleOptionsChanged
  (bind _creativegameuinewhud (Class|GetClassDefaults "/FortniteCoreUI/Settings/ExperienceSettingsComponent.ExperienceSettingsComponent_C"))
  (if (ScalableFloat|GetValueasBool (Variables|Default|GetHF_ForceEnableInEdit))
    _
    (elif (Variables|Default|GetbUseLatestHUD)
      _
      (else
        (ScalableFloat|SetValuewithContextParam (Components|GetOwner) _creativegameuinewhud (Math|Conversions|ToFloat(Boolean) false))
        (Development|PrintString "Creative/UEFN: Disable New CH7 HUD")
        (CallFunction|RefreshHUD))))
  (ScalableFloat|SetValuewithContextParam (Components|GetOwner) _creativegameuinewhud (Math|Conversions|ToFloat(Boolean) true))
  (Development|PrintString "Creative/UEFN: Enable New CH7 HUD"))

(fn RefreshHUD
  (bind _asdevice_experience_settings_v2 (Utilities|Casting|CastToDevice_ExperienceSettings_V2 (Components|GetOwner)))
  (Utilities|IsValid (Class|DeviceExperienceSettingsV2|GetCurrentMinigame _asdevice_experience_settings_v2)
    (:Is%20Valid
      (Utilities|IsValid (Actor|GetComponentbyClass (Class|DeviceExperienceSettingsV2|GetCurrentMinigame _asdevice_experience_settings_v2) "/Script/FortniteGame.FortMutatorListComponent")
        (:Is%20Valid
          (Utilities|IsValid (Modifier|GetMutatorbyClass (Actor|GetComponentbyClass (Class|DeviceExperienceSettingsV2|GetCurrentMinigame _asdevice_experience_settings_v2) "/Script/FortniteGame.FortMutatorListComponent") "/Game/Athena/Playlists/Creative/Mutators/CreativeMutator_NewHUD.CreativeMutator_NewHUD_C")
            (:Is%20Valid
              (bind _asfort_athena_mutator_new_hud (Utilities|Casting|CastToFortAthenaMutator_NewHUD (Modifier|GetMutatorbyClass (Actor|GetComponentbyClass (Class|DeviceExperienceSettingsV2|GetCurrentMinigame _asdevice_experience_settings_v2) "/Script/FortniteGame.FortMutatorListComponent") "/Game/Athena/Playlists/Creative/Mutators/CreativeMutator_NewHUD.CreativeMutator_NewHUD_C")))
              (Class|FortAthenaMutatorNewHUD|TriggerHUDRefresh _asfort_athena_mutator_new_hud))))))))
