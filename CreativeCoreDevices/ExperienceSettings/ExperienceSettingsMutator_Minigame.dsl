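; Experience settings mutator for minigames that supplies HUD visibility tags for a pawn based on the minigame's settings.
; GetHUDVisibilityTags looks up the pawn's minigame settings machine, checking that both pawn and machine are valid.
; When that machine disables squad quick chat, it returns a container holding the touch QuickChat button's HUD tag.
; The construction script is empty.
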(fn GetHUDVisibilityTags
  (params (Pawn :type Pawn%20Object%20Reference))
  (bind _returnvalue (Minigame|GetMinigameSettingsMachine Pawn))
  (Utilities|IsValid Pawn
    (:Is%20Valid
      (Utilities|IsValid _returnvalue
        (:Is%20Valid
          (if (Variables|/Script/FortniteGame.FortMinigameSettingsBuilding|GetbDisableSquadQuickChat _returnvalue)
            (return (GameplayTags|MakeGameplayTagContainerfromTag "(TagName=\"HUD.Bacchus.TouchButton.QuickChat\")"))
            (else
              _)))
        (:Is%20Not%20Valid
          _))
      (return 0))
    (:Is%20Not%20Valid
      _)))

(fn ConstructionScript)
