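; Helper functions for the experience settings that convert mantling ledge heights between float values and byte preset indices, and migrate old float options.
; Float-to-preset clamps to the min/max in the conversion params, then floors the offset from min by the step increment; preset-to-float rebuilds the height, capped at max.
; Migration converts only non-default old values, resets the by-ref variable to its default, and returns a migrated flag with the converted preset.
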(fn ConvertFloatToMantlingHeightPreset
  (params (Value :type Float%20%28double-precision%29) (Params_MinLedgeHeight_7_1BE18C2B485B6B517841749D95076B2C :type Float%20%28double-precision%29) (Params_MaxLedgeHeight_8_90FFC5AC42D39345E45E6E975893BAE0 :type Float%20%28double-precision%29) (Params_StepIncrement_6_FFAFFA3B43B9584FE717CEB9DAE0FE43 :type Float%20%28double-precision%29))
  (return (Math|Conversions|ToByte(Integer) (Math|Float|Floor (/ (- (Math|Float|Clamp(Float) Value Params_MinLedgeHeight_7_1BE18C2B485B6B517841749D95076B2C Params_MaxLedgeHeight_8_90FFC5AC42D39345E45E6E975893BAE0) Params_MinLedgeHeight_7_1BE18C2B485B6B517841749D95076B2C) Params_StepIncrement_6_FFAFFA3B43B9584FE717CEB9DAE0FE43)))))

(fn ConvertMantlingHeightPresetToFloat
  (params (Value :type Byte) (Params_MinLedgeHeight_7_1BE18C2B485B6B517841749D95076B2C :type Float%20%28double-precision%29) (Params_MaxLedgeHeight_8_90FFC5AC42D39345E45E6E975893BAE0 :type Float%20%28double-precision%29) (Params_StepIncrement_6_FFAFFA3B43B9584FE717CEB9DAE0FE43 :type Float%20%28double-precision%29))
  (return (Math|Float|Min(Float) (+ Params_MinLedgeHeight_7_1BE18C2B485B6B517841749D95076B2C (* Params_StepIncrement_6_FFAFFA3B43B9584FE717CEB9DAE0FE43 (Math|Conversions|ToFloat(Byte) Value))) Params_MaxLedgeHeight_8_90FFC5AC42D39345E45E6E975893BAE0)))

(fn MigrateOptionType_FloatToMantlingHeightPreset
  (params (OldOptionVariable :type Float%20%28double-precision%29%20%28by%20ref%29) (OldOptionDefault :type Float%20%28double-precision%29) (Params :type Mantling%20Conversion%20Params%20Structure))
  (Utilities|FlowControl|Sequence
    (:then_0
      (if (!= OldOptionVariable OldOptionDefault)
        (LocalVariables|SetConvertedOptionValue (CallFunction|ConvertFloattoMantlingHeightPreset OldOptionVariable Params))
        (Variables|SetBy-RefVar OldOptionVariable OldOptionDefault)
        (LocalVariables|SetbWasMigrated true)))
    (:then_1
      (return (LocalVariables|GetbWasMigrated) (LocalVariables|GetConvertedOptionValue)))))
