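; Experience Settings device component that manages bot test players for moderators and Creative/edit playlists.
; On user option load or change, the authority sets the single-user multiplayer playspace component's test player count from None, Fill or Custom settings.
; A spawn delegate, reset at minigame end, async-loads the character definition and spawns bots using requester pawn or air-spawn locations.
; A participant definition overrides behavior paths chosen by debug override or option, with Newton variants for physically simulated islands.
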(event Custom|RegisterPlayspaceEvents
  (bind _returnvalue (CallFunction|IsPlayerAModerator))
  (bind _returnvalue_2 (Fortnite|Athena|GameState|GetGameStateAthena))
  (bind _returnvalue_3 (Class|FortGameStateAthena|IsVKEditPlaylist _returnvalue_2))
  (bind _returnvalue_4 (Class|FortGameStateAthena|IsAnyCreativePlaylist _returnvalue_2))
  (bind _returnvalue_5 (or _returnvalue_3 _returnvalue_4))
  (bind _returnvalue_6 (or _returnvalue _returnvalue_5))
  (bind _asdevice_experience_settings_v2 (Utilities|Casting|CastToDevice_ExperienceSettings_V2 (Components|GetOwner)))
  (Utilities|IsValid (Class|DeviceExperienceSettingsV2|GetCurrentMinigame _asdevice_experience_settings_v2)
    (:Is%20Valid
      (Minigame|BindEventtoMinigameEndedDelegate (Class|DeviceExperienceSettingsV2|GetCurrentMinigame _asdevice_experience_settings_v2) (EventDispatchers|CreateEvent :!event OnMinigameEnd)))
    (:Is%20Not%20Valid
      _))
  (if _returnvalue_6
    (switch Utilities|FlowControl|Switch|SwitchonETestPlayersSetting (Variables|UserOptions|GetTestPlayersOnStart)
      (:None
        (if _returnvalue_6
          _))
      (:Fill
        _)
      (:Custom
        _))
    (if (not (Asset|IsAssetNull (CallFunction|GetTestPlayerDefinition)))
      (bind _returnvalue_7 (Class|PlayspaceLibrary|GetPlayspaceforActor (Components|GetOwner) (Components|GetOwner)))
      (Utilities|IsValid _returnvalue_7
        (:Is%20Valid
          (Utilities|IsValid (Actor|GetComponentbyClass _returnvalue_7 "/Script/FortniteGame.FortPlayspaceComponent_SingleUserMultiplayer")
            (:Is%20Valid
              (Class|FortPlayspaceComponentSingleUserMultiplayer|SetSpawnBotsDelegate (Actor|GetComponentbyClass _returnvalue_7 "/Script/FortniteGame.FortPlayspaceComponent_SingleUserMultiplayer") (EventDispatchers|CreateEvent :!event Custom|SpawnBotsEvent))
              (Utilities|AsyncLoadAsset (CallFunction|GetParticipantDefinition)))))))))

(event Custom|UnregisterPlayspaceEvents
  (CallFunction|ResetonSpawnEventDelegate)
  (bind _asdevice_experience_settings_v2 (Utilities|Casting|CastToDevice_ExperienceSettings_V2 (Components|GetOwner)))
  (Utilities|IsValid (Class|DeviceExperienceSettingsV2|GetCurrentMinigame _asdevice_experience_settings_v2)
    (:Is%20Valid
      (Minigame|UnbindallEventsfromMinigameEndedDelegate (Class|DeviceExperienceSettingsV2|GetCurrentMinigame _asdevice_experience_settings_v2)))))

(event Custom|SpawnBotsEvent
  (params (Count :type Integer) (Outer :type Object%20Reference))
  (bind _returnvalue (Class|PlayspaceLibrary|GetPlayspaceforActor (Components|GetOwner) (Components|GetOwner)))
  (for _ (range 1 (+ Count 1))
    (bind (_characterdefinition _componentslist) (AISpawner|LoadAndCreateComponentListAsync (CallFunction|GetTestPlayerDefinition) (Utilities|Array|MakeArray 0))
      (:OnFinished
        (bind _returnvalue_2 (Class|PlayspaceLibrary|GetPlayspaceforActor (Components|GetOwner) (Components|GetOwner)))
        (Utilities|Casting|CastToFortPlayspace _returnvalue_2)
        (Class|FortPlayspaceComponentSingleUserMultiplayer|SetAnalyticsCurrentBehaviorScriptType (Actor|GetComponentbyClass _returnvalue "/Script/FortniteGame.FortPlayspaceComponent_SingleUserMultiplayer") (select (Utilities|IsValidSoftObjectReference (CallFunction|GetParticipantDefinition)) "Custom" (Utilities|String|EnumtoString (Variables|UserOptions|GetTestPlayerDefinition))))
        (Utilities|IsValid (Pawn|GetControlledPawn (Variables|/Script/FortniteGame.FortPlayspaceComponent_SingleUserMultiplayer|GetCachedBotOwnerController (Actor|GetComponentbyClass _returnvalue "/Script/FortniteGame.FortPlayspaceComponent_SingleUserMultiplayer")))
          (:Is%20Valid
            ;[1 1 1] Initial spawn on the Player Controller's Pawn location requesting the Test Player
            (Variables|Default|SetDefaultSpawnLocation (Transformation|GetActorLocation (Pawn|GetControlledPawn (Variables|/Script/FortniteGame.FortPlayspaceComponent_SingleUserMultiplayer|GetCachedBotOwnerController (Actor|GetComponentbyClass _returnvalue "/Script/FortniteGame.FortPlayspaceComponent_SingleUserMultiplayer"))))))
          (:Is%20Not%20Valid
            (bind _returnvalue_3 (Minigame|GetMinigameforActor (Variables|/Script/FortniteGame.FortPlayspaceComponent_SingleUserMultiplayer|GetCachedBotOwnerController (Actor|GetComponentbyClass _returnvalue "/Script/FortniteGame.FortPlayspaceComponent_SingleUserMultiplayer"))))
            (Utilities|IsValid _returnvalue_3
              (:Is%20Valid
                (Utilities|IsValid (Minigame|GetPlayerSpawnComponent _returnvalue_3)
                  (:Is%20Valid
                    (bind _asfort_player_state_athena (Utilities|Casting|CastToFortPlayerStateAthena (Variables|/Script/Engine.Controller|GetPlayerState (Variables|/Script/FortniteGame.FortPlayspaceComponent_SingleUserMultiplayer|GetCachedBotOwnerController (Actor|GetComponentbyClass _returnvalue "/Script/FortniteGame.FortPlayspaceComponent_SingleUserMultiplayer"))))
                      (:then
                        (bind _outlocation (Class|FortMinigamePlayerSpawnComponent|GetAirSpawnLocation (Minigame|GetPlayerSpawnComponent _returnvalue_3) _asfort_player_state_athena))
                        (Variables|Default|SetDefaultSpawnLocation _outlocation))
                      (:CastFailed
                        _))
                    (Class|FortPlayspaceComponentSingleUserMultiplayer|SpawnTestPlayer (Actor|GetComponentbyClass _returnvalue "/Script/FortniteGame.FortPlayspaceComponent_SingleUserMultiplayer") _componentslist (Math|Conversions|ToTransform(Vector) (Variables|Default|GetDefaultSpawnLocation)) (Variables|/Script/FortniteGame.FortPlayspaceComponent_SingleUserMultiplayer|GetCachedBotOwnerController (Actor|GetComponentbyClass _returnvalue "/Script/FortniteGame.FortPlayspaceComponent_SingleUserMultiplayer"))))))
              (:Is%20Not%20Valid
                _))))))))

(event UserOptions|EventHandleUserOptionsLoaded
  ;[1 1 1] Make sure any time we update this component user options, we update the required test player count
  (|CallParentFunction)
  (Default|UpdateTestPlayers))

(event UserOptions|EventHandleUserOptionsChanged
  (params (ChangedPropertyNames :type Array%20of%20Names))
  ;[1 1 1] Make sure any time we update this component user options, we update the required test player count
  (|CallParentFunction ChangedPropertyNames)
  (Default|UpdateTestPlayers))

(fn DoesParticipantRequireNavmesh
  (if (Utilities|IsValidSoftObjectReference (CallFunction|GetTestPlayerDefinition))
    (switch Utilities|FlowControl|Switch|SwitchonETestPlayersSetting (Variables|UserOptions|GetTestPlayersOnStart)
      (:None
        _)
      (:Fill
        (if (or (Class|FortGameStateAthena|IsVKEditPlaylist (Fortnite|Athena|GameState|GetGameStateAthena)) (Class|FortGameStateAthena|IsAnyCreativePlaylist (Fortnite|Athena|GameState|GetGameStateAthena)))
          (return true)
          (else
            _)))
      (:Custom
        _))
    (return false)
    (else
      (return false))))

(fn GetDebugTestPlayerBehavior
  (bind _singleplayerplayspacecomponent (LocalVariables|GetSinglePlayerPlayspaceComponent))
  (bind _overridennexttestplayerbehavior (Variables|/Script/FortniteGame.FortPlayspaceComponent_SingleUserMultiplayer|GetOverridenNextTestPlayerBehavior _singleplayerplayspacecomponent))
  (bind _returnvalue (Class|PlayspaceLibrary|GetPlayspaceforActor (Components|GetOwner) (Components|GetOwner)))
  (bind _output_get (LocalVariables|SetSinglePlayerPlayspaceComponent (Actor|GetComponentbyClass _returnvalue "/Script/FortniteGame.FortPlayspaceComponent_SingleUserMultiplayer")))
  (Utilities|IsValid _output_get
    (:Is%20Valid
      (if (>= _overridennexttestplayerbehavior 0)
        (return true (Utilities|Enum|BytetoEnumEnum_ExperienceSettings_TestPlayerDefinition (Math|Conversions|ToByte(Integer) _overridennexttestplayerbehavior)))
        (else
          _)))
    (:Is%20Not%20Valid
      _))
  (return false "NewEnumerator3"))

(fn GetTestPlayerDefinition
  (bind _testplayerbehaviorid (LocalVariables|GetTestPlayerBehaviorId))
  (bind _picknewtoncharacters (LocalVariables|GetPickNewtonCharacters))
  (bind _pathtofollowplayer (Variables|Default|GetPathToFollowPlayer))
  (bind _pathtofollowplayer_newton (Variables|Default|GetPathToFollowPlayer_Newton))
  (bind _returnvalue (Utilities|String|SelectString _pathtofollowplayer_newton _pathtofollowplayer _picknewtoncharacters))
  (bind _picknewtoncharacters_2 (LocalVariables|GetPickNewtonCharacters))
  (bind _pathtorandommovement (Variables|Default|GetPathToRandomMovement))
  (bind _pathtorandommovement_newton (Variables|Default|GetPathToRandomMovement_Newton))
  (bind _returnvalue_2 (Utilities|String|SelectString _pathtorandommovement_newton _pathtorandommovement _picknewtoncharacters_2))
  (bind _picknewtoncharacters_3 (LocalVariables|GetPickNewtonCharacters))
  (bind _pathtoidle (Variables|Default|GetPathToIdle))
  (bind _pathtoidle_newton (Variables|Default|GetPathToIdle_Newton))
  (bind _returnvalue_3 (Utilities|String|SelectString _pathtoidle_newton _pathtoidle _picknewtoncharacters_3))
  (bind _returnvalue_4 (Utilities|Select :Index _testplayerbehaviorid :NewEnumerator3 _returnvalue_3 :NewEnumerator1 _returnvalue_2 :NewEnumerator0 _returnvalue))
  (bind _returnvalue_5 (Utilities|MakeSoftObjectPath _returnvalue_4))
  (bind _returnvalue_6 (CallFunction|GetParticipantDefinition))
  (bind _returnvalue_7 (CharacterDefinition|ToCharacterDefinitionSoftObjectReference _returnvalue_5))
  (if (Utilities|IsValidSoftObjectReference _returnvalue_6)
    (return _returnvalue_6)
    (else
      (bind (_overriden _debugbehavior) (CallFunction|GetDebugTestPlayerBehavior))
      (if _overriden
        (LocalVariables|SetTestPlayerBehaviorId _debugbehavior)
        (else
          (LocalVariables|SetTestPlayerBehaviorId (Variables|UserOptions|GetTestPlayerDefinition))))
      (LocalVariables|SetPickNewtonCharacters (Fortnite|IsPhysicallySimulatedIsland))
      (if (not (Utilities|String|IsEmpty _returnvalue_4))
        (if (Utilities|IsValidSoftObjectReference _returnvalue_7)
          (return _returnvalue_7)
          (else
            _)))
      (return 0))))

(fn OnMinigameEnd
  (CallFunction|ResetonSpawnEventDelegate))

(fn ResetOnSpawnEventDelegate
  (bind _returnvalue (Class|PlayspaceLibrary|GetPlayspaceforActor (Components|GetOwner) (Components|GetOwner)))
  (Utilities|IsValid _returnvalue
    (:Is%20Valid
      (Utilities|IsValid (Actor|GetComponentbyClass _returnvalue "/Script/FortniteGame.FortPlayspaceComponent_SingleUserMultiplayer")
        (:Is%20Valid
          (Class|FortPlayspaceComponentSingleUserMultiplayer|ResetSpawnBotsDelegate (Actor|GetComponentbyClass _returnvalue "/Script/FortniteGame.FortPlayspaceComponent_SingleUserMultiplayer")))))))

(fn UpdateMaxTestPlayers
  (params (MaxPlayers :type Integer))
  (if (> (Variables|UserOptions|GetNumberOfTestPlayers) MaxPlayers)
    (Variables|UserOptions|SetNumberOfTestPlayers MaxPlayers)))

(fn UpdateTestPlayers
  (bind _ownerdevice (LocalVariables|GetOwnerDevice))
  (bind _maxplayers (Variables|/Script/FortniteGame.FortMinigameSettingsBuilding|GetMaxPlayers _ownerdevice))
  (bind _asdevice_experience_settings_v2 (Utilities|Casting|CastToDevice_ExperienceSettings_V2 (Components|GetOwner)))
  (bind _output_get (LocalVariables|SetOwnerDevice _asdevice_experience_settings_v2))
  (if (Networking|HasAuthority _asdevice_experience_settings_v2)
    (bind _returnvalue (Class|PlayspaceLibrary|GetPlayspaceforActor _output_get _output_get))
    (bind _output_get_2 (LocalVariables|SetCurrentPlayspace _returnvalue))
    (if (Utilities|IsValid _output_get_2 :!class /Script/Engine.KismetSystemLibrary)
      (bind _outcontrollers (Class|PlayspaceLibrary|GetAllControllers _output_get_2))
      (bind _array_element (Utilities|Array|ForEachLoopwithBreak _outcontrollers)
        (:LoopBody
          (bind _asfort_player_controller_gameplay (Utilities|Casting|CastToFortPlayerControllerGameplay _array_element))
          (LocalVariables|SetFirstPlayspaceLocalPlayer _asfort_player_controller_gameplay))
        (:Completed
          _))
      (if (Utilities|IsValid (LocalVariables|GetFirstPlayspaceLocalPlayer) :!class /Script/Engine.KismetSystemLibrary)
        (bind _output_get_3 (LocalVariables|SetSingleUserComponent (Actor|GetComponentbyClass (LocalVariables|GetCurrentPlayspace) "/Script/FortniteGame.FortPlayspaceComponent_SingleUserMultiplayer")))
        (Utilities|IsValid _output_get_3
          (:Is%20Valid
            (if (Variables|/CreativeCoreDevices/Device_ExperienceSettings_V2.Device_ExperienceSettings_V2_C|GetbDebugSettingsEnabled _ownerdevice)
              (switch Utilities|FlowControl|Switch|SwitchonETestPlayersSetting (Variables|UserOptions|GetTestPlayersOnStart)
                (:None
                  _)
                (:Fill
                  (Class|FortPlayspaceComponentSingleUserMultiplayer|SetTestPlayerCount (LocalVariables|GetSingleUserComponent) (LocalVariables|GetFirstPlayspaceLocalPlayer) (Math|Conversions|ToByte(Integer) _maxplayers)))
                (:Custom
                  (Class|FortPlayspaceComponentSingleUserMultiplayer|SetTestPlayerCount (LocalVariables|GetSingleUserComponent) (LocalVariables|GetFirstPlayspaceLocalPlayer) (Math|Conversions|ToByte(Integer) (Math|Integer|Clamp(Integer) (Variables|UserOptions|GetNumberOfTestPlayers) 0 _maxplayers)))))
              (else
                _))
            (Class|FortPlayspaceComponentSingleUserMultiplayer|SetTestPlayerCount (LocalVariables|GetSingleUserComponent) (LocalVariables|GetFirstPlayspaceLocalPlayer) 0)))))))
